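# Translucent unlit Niagara ribbon material for the IronLilac skydiving contrail, drawing a wavy paint trail with ink clouds.
# Panned, sine-waved UVs sample noise that erodes a width-masked strip with a rounded head and fading tail.
# An SDF texture, thresholded along the trail, adds clouds outside the core, scaled by a dynamic parameter.
# Noise and particle color drive emissive, particle alpha and depth fade shape opacity, and dynamic parameters offset each trail.
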Material /BRCosmetics/Athena/Effects/Particles/Skydiving/NS_Contrail_IronLilac/M_Contrail_IronLilac_PaintTrail
graph {
 properties {
  bCastRayTracedShadows=false
  bUsedWithNiagaraRibbons=true
  blendMode=BLEND_Translucent
  opacityMaskClipValue=0.01
  pixelDepthOffsetMode=PDOM_Legacy
  shadingModel=MSM_Unlit
  twoSided=true
  usedShadingModels=Unlit
 }
 nodes [
  n1 = NamedRerouteUsage()
  n2 = TextureCoordinate()
  n3 = ComponentMask(None=@n2) {r=true}
  n4 = ScalarParameter() {defaultValue=0.2, group=Paint Trail, parameterName=Width, sliderMax=1}
  n5 = SphereMask(
   A=AppendVector(A=Power(Base=@n3) {constExponent=1.5}, B=ComponentMask(None=@n2) {g=true})
   B=AppendVector(A=@n4, B=Constant() {r=0.5})
   Radius=@n4
  ) {hardnessPercent=80}
  n6 = SphereMask(
   A=ComponentMask(None=NamedRerouteUsage()) {g=true}
   B=Constant() {r=0.5}
   Radius=ScalarParameter() {defaultValue=0.2, group=Paint Trail, parameterName=Width, sliderMax=1}
   Hardness=Constant()
  )
  n7 = TextureCoordinate()
  n8 = OneMinus(None=Power(Base=ComponentMask(None=@n7) {r=true}) {constExponent=5})
  n9 = TextureCoordinate()
  n10 = Add(
   A=DynamicParameter() {defaultValue=<rgba 0 0 1 1>, paramNames [Random, Offset X, Ink Cloud Opacity, Param4]}.Random
   B=Power(Base=ComponentMask(None=@n9) {r=true}, Exp=ScalarParameter() {defaultValue=1, group=Paint Trail, parameterName=UV Stretch, sliderMax=2})
  )
  n11 = ComponentMask(None=@n9) {g=true}
  n12 = ScalarParameter() {defaultValue=-0.5, group=Paint Trail, parameterName=Speed X, sliderMax=5, sliderMin=-5}
  n13 = Constant()
  n14 = Add(
   A=Panner(Coordinate=AppendVector(A=@n10, B=@n11), Speed=AppendVector(A=@n12, B=@n13)) {speedX=0.5}
   B=AppendVector(A=DynamicParameter() {defaultValue=<rgba 0 0 1 1>, paramNames [Random, Offset X, Ink Cloud Opacity, Param4]}.Offset X, B=Constant())
  )
  n15 = Power(
   Base=ComponentMask(None=TextureCoordinate()) {r=true}
   Exp=ScalarParameter() {defaultValue=1, group=Paint Trail, parameterName=UV Stretch, sliderMax=2}
  )
  n16 = Multiply(A=Time(), B=ScalarParameter() {defaultValue=0.5, group=Wave, parameterName=Wave Speed, sliderMax=5, sliderMin=-5})
  n17 = Add(
   A=Add(A=Add(A=@n15, B=@n16), B=DynamicParameter() {defaultValue=<rgba 0 0 1 1>, paramNames [Random, Offset X, Ink Cloud Opacity, Param4]}.Random)
   B=DynamicParameter() {defaultValue=<rgba 0 0 1 1>, paramNames [Random, Offset X, Ink Cloud Opacity, Param4]}.Offset X
  )
  n18 = TextureCoordinate()
  n19 = TextureCoordinate()
  n20 = Multiply(
   A=Multiply(A=Sine(None=@n17), B=ComponentMask(None=@n18) {r=true}) {constB=0.2}
   B=Power(Base=ComponentMask(None=@n19) {r=true}, Exp=ScalarParameter() {defaultValue=3, group=Wave, parameterName=Wave Gradient, sliderMax=10})
  ) {constB=0.2}
  n21 = ScalarParameter() {defaultValue=0.5, group=Wave, parameterName=Wave Intensity, sliderMax=1}
  n22 = AppendVector(A=ComponentMask(None=@n14) {r=true}, B=Add(A=ComponentMask(None=@n14) {g=true}, B=Multiply(A=@n20, B=@n21) {constB=0.3}))
  n23 = ScalarParameter() {defaultValue=1, group=Paint Trail, parameterName=Tiling X, sliderMax=5}
  n24 = ScalarParameter() {defaultValue=1, group=Paint Trail, parameterName=Tiling Y, sliderMax=5}
  n25 = TextureSample(UVs=Multiply(A=NamedRerouteDeclaration(None=@n22) {name=Final UV, nodeColor=<rgba 0.2422 1 0 1>}, B=AppendVector(A=@n23, B=@n24))) {samplerType=SAMPLERTYPE_Grayscale, texture=&LowResBlurredNoise}
  n26 = ScalarParameter() {defaultValue=2, group=Paint Trail, parameterName=Noise Contrast}
  n27 = Subtract(
   A=Multiply(A=Multiply(A=@n6, B=@n8), B=ScalarParameter() {defaultValue=2, group=Paint Trail, parameterName=Base Opacity, sliderMax=10})
   B=Multiply(A=Power(Base=@n25.R, Exp=@n26) {constExponent=1.25}, B=ScalarParameter() {defaultValue=5, group=Paint Trail, parameterName=Noise Intensity}) {constB=2}
  )
  n28 = SmoothStep(Value=ComponentMask(None=TextureCoordinate()) {r=true}) {constMax=0.25, constMin=0.1}
  n29 = SmoothStep(
   Min=ScalarParameter() {group=Paint Trail, parameterName=OpacityMin, sliderMax=1}
   Max=ScalarParameter() {defaultValue=1, group=Paint Trail, parameterName=Opacity Max, sliderMax=1}
   Value=Saturate(None=LinearInterpolate(A=@n5, B=@n27, Alpha=@n28))
  )
  n30 = ParticleColor()
  DepthFade(
   Opacity=Multiply(A=Add(A=@n1, B=@n29), B=@n30.A) {constB=0.5}
   FadeDistance=ScalarParameter() {defaultValue=10, parameterName=DepthFadeDistance, sliderMax=50}
  )
  n31 = Saturate(None=OneMinus(None=Power(Base=@n25.R) {constExponent=0.25}))
  n32 = Saturate(None=Multiply(A=@n30.RGB, B=Add(A=@n31) {constB=0.25}))
  n33 = ComponentMask(None=NamedRerouteUsage()) {g=true}
  n34 = Constant() {r=0.5}
  n35 = ScalarParameter() {defaultValue=0.2, group=Paint Trail, parameterName=Width, sliderMax=1}
  n36 = Add(
   A=ScalarParameter() {defaultValue=-0.5, group=Clouds, parameterName=Cloud Speed X, sliderMax=5, sliderMin=-5}
   B=Multiply(A=ScalarParameter() {defaultValue=-0.5, group=Paint Trail, parameterName=Speed X, sliderMax=5, sliderMin=-5}) {constB=-1}
  )
  n37 = Constant()
  n38 = MaterialFunctionCall(XY=Panner(Coordinate=NamedRerouteUsage(), Speed=AppendVector(A=@n36, B=@n37)) {speedX=0.5}) {materialFunction=&Swizzle}
  n39 = AppendVector(A=Constant() {r=1}, B=ScalarParameter() {defaultValue=-2, group=Clouds, parameterName=Tiling Y Clouds, sliderMax=5, sliderMin=-5})
  n40 = TextureCoordinate()
  n41 = Multiply(
   A=ScalarParameter() {defaultValue=1, group=Clouds, parameterName=Tiling X Clouds, sliderMax=5, sliderMin=-5}
   B=SmoothStep(Value=ComponentMask(None=@n40) {r=true}) {constMax=0.2}
  )
  n42 = Constant() {r=1}
  n43 = TextureCoordinate()
  n44 = TextureSample!(
   UVs=MaterialFunctionCall(UVs=Multiply(A=@n38.YX, B=@n39), Texture Scale=AppendVector(A=@n41, B=@n42)) {materialFunction=&ScaleUVsByCenter}.UVs
   Tex
   "DDX(UVs)"=DDX(Value=@n43)
   "DDY(UVs)"=DDY(Value=@n43)
   Apply View MipBias
  ) {
   mipValueMode=TMVM_Derivative
   samplerType=SAMPLERTYPE_LinearGrayscale
   texture=&/T_Contrail_IronLilac_G_SDF
  }
  n45 = ComponentMask(None=TextureCoordinate()) {r=true}
  n46 = Multiply(A=Power(Base=SmoothStep(Value=@n45) {constMin=0.3}) {constExponent=3}) {constB=0.25}
  n47 = ScalarParameter() {defaultValue=0.4, group=Clouds, parameterName=Cloud Size, sliderMax=2}
  n48 = Divide(
   A=Subtract(A=@n44.R, B=Add(A=@n46, B=@n47) {constB=0.2}) {constB=0.5}
   B=ScalarParameter() {defaultValue=1, group=Clouds, parameterName=Cloud Fade, sliderMax=2}
  ) {constB=0.5}
  n49 = Multiply(
   A=OneMinus(None=SphereMask(A=@n33, B=@n34, Radius=@n35) {attenuationRadius=0.2, hardnessPercent=0})
   B=Multiply(
    A=DynamicParameter() {defaultValue=<rgba 0 0 1 1>, paramNames [Random, Offset X, Ink Cloud Opacity, Param4]}.Ink Cloud Opacity
    B=Saturate(None=@n48)
   )
  )
  MaterialFunctionCall(Emissive=LinearInterpolate(
    A=@n32
    B=LinearInterpolate(A=@n30.RGB, B=@n32) {constAlpha=0.25}
    Alpha=NamedRerouteDeclaration(None=@n49) {name=Ink Cloud, nodeColor=<rgba 0 1 0.9063 1>}
   )) {materialFunction=&MF_FX_EmissiveAdjustments}
 ]
}
