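# Niagara skydiving contrail for the IronLilac cosmetic, attaching cyan-blue trails, glows and spinning pen meshes to character sockets.
# Center and Paint ribbons each build four five-point trails, placing particles procedurally from sockets along an interpolated look-at vector.
# Splash foam sprites ride the Paint ribbons via a particle reader, while Decal_Glow adds sprites and lights at the sockets.
# Dive values drive ribbon material scrolling, paint length, glow size and foam rate.
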NiagaraSystem /BRCosmetics/Athena/Effects/Particles/Skydiving/NS_Contrail_IronLilac/NS_Contrail_IronLilac
dependencies {
 usedDataInterfaces [
  &NiagaraDataInterfaceArrayFloat
  &NiagaraDataInterfaceArrayFloat3
  &NiagaraDataInterfaceCurve
  &NiagaraDataInterfaceColorCurve
  &NiagaraDataInterfaceArrayInt32
  &NiagaraDataInterfaceParticleRead
  &NiagaraDataInterfaceSocketReader
 ]
 usedDynamicInputs [
  &FN_FrontEndVector3Modifier
  &Multiply_Vector
  &Lerp_Vector
  &TransformVector
  &NormalizeVector
  &TransformPosition
 ]
 usedModules [
  &self:SystemSpawnScript.SystemScriptSource.SystemScriptGraph.NiagaraNodeAssignment_2.SetVariables_#1
  &SystemState
  &self:SystemSpawnScript.SystemScriptSource.SystemScriptGraph.NiagaraNodeAssignment_3.SetVariables_#2
  &self:SystemSpawnScript.SystemScriptSource.SystemScriptGraph.NiagaraNodeAssignment_4.SetVariables_#3
  &self:SystemSpawnScript.SystemScriptSource.SystemScriptGraph.NiagaraNodeAssignment_5.SetVariables_#4
 ]
}
emitters {
 Center_Ribbons {
  emitterSpawnScript [
   /SetVariables(Emitter.InterpolatedVector:Vector3f=<0 0 1>, Emitter.TargetVector:Vector3f=<0 0 1>) {isSetParameters=true, module=SetVariables_#5}
   /SetVariables(Emitter.RibbonLength=222) {isSetParameters=true, module=SetVariables_#6}
   /SetVariables(Emitter.ParticlesPerTrail=5) {isSetParameters=true, module=SetVariables_#7}
  ]
  emitterUpdateScript [
   &EmitterState(
    Inactive Response=::NewEnumerator1
    Life Cycle Mode=::NewEnumerator1
    Loop Behavior=::NewEnumerator1
    Loop Duration Mode=::NewEnumerator1
    UseLoopDelay=false
   )
   &SpawnBurst_Instantaneous(
    Spawn Count=&Multiply_Int(A=4, B=$Emitter.ParticlesPerTrail)
    Use Spawn Probability=false
    UseLoopCountLimit=false
   )
   &self:ScratchModule(Target Vector:Vector3f=$System.LookAtVector) {module=ScratchModule}
  ]
  particleSpawnScript [
   &InitializeParticle(
    Lifetime=2
    Position Mode=::NewEnumerator0
    Ribbon Width=15
    Ribbon Width Mode=::NewEnumerator1
    Sprite Size Mode=::NewEnumerator3
    Uniform Sprite Size=5
    Write Material Random=true
   )
   &PartitionParticles(
    Normalized Index In Partition Includes One=true
    Number Of Partitions=4
    Particles Per Partition:NiagaraInt32=$Emitter.ParticlesPerTrail
    Partition Mode:ENiagara_PartitionMode=::NewEnumerator1
    Write Partition Index to Ribbon Link Order:ENiagara_AttributeSamplingApplyOutput=::NewEnumerator0
    Write Partition to Ribbon ID=::NewEnumerator0
   )
   /SetVariables(Particles.RibbonWidth=&SelectFloatFromArray(Direct Array Index=$Particles.PartitionParticles.Partition, Float Selection Array=DataInterface() {propertyValues {FloatData [80,80,40,40]}})) {isSetParameters=true, module=SetVariables_#8}
   /SetVariables(Particles.RibbonLength:NiagaraFloat=&SelectFloatFromArray(
     Direct Array Index=$Particles.PartitionParticles.Partition
     Float Selection Array=DataInterface() {propertyValues {FloatData [180,180,100,100]}}
    )) {isSetParameters=true, module=SetVariables_#9}
  ]
  particleUpdateScript [
   &ParticleState(Kill Particles When Lifetime Has Elapsed=false)
   &SocketLocation(
    Orientation Sampling=::NewEnumerator1
    Position Sampling=::NewEnumerator1
    SamplePosOffset=true
    Sampled Position Offset=&SelectVectorFromArray(
     Direct Array Index=$Particles.PartitionParticles.Partition
     Vector Selection Array=DataInterface() {propertyValues {FloatData [<0 0 26>, <0 0 26>, <0 0 -5>, <0 0 -5>]}}
    )
    Sampled Position Offset Coordinate Space:ENiagaraSkelMeshTransforms=::NewEnumerator2
    Socket Index=$Particles.PartitionParticles.Partition
    Socket Reader=$System.Socket Reader
    Socket Sampling Mode=::NewEnumerator3
    VelocitySpeedLimit=false
   )
   /SetVariables(Particles.RibbonLength:NiagaraFloat=&Multiply_Float(
     A=$Particles.Initial.RibbonLength
     B=&FloatFromCurve(FloatCurve=DataInterface() {propertyValues {Curve {keys [{interpMode=RCIM_Cubic, tangentMode=RCTM_User, time=0, value=0}, {interpMode=RCIM_Cubic, tangentMode=RCTM_User, time=1, value=1}]}}})
    )) {isSetParameters=true, module=SetVariables_#10}
   /SetVariables(Particles.InterpolatedTrailVector:Vector3f=$Emitter.InterpolatedVector) {isSetParameters=true, module=SetVariables_#11}
   /SetVariables(Particles.InterpolatedPosition:NiagaraPosition=&AddVectorToPosition(
     Position=$Output.SocketLocation.SampledPosition
     Vector=&Multiply_VectorByFloat(
      Vector=$Particles.InterpolatedTrailVector
      Float=&Multiply_Float(A=$Particles.RibbonLength, B=$Particles.PartitionParticles.Normalized Index In Partition)
     )
    )) {isSetParameters=true, module=SetVariables_#12}
   &ScaleRibbonWidth(Ribbon Width Scale=&Multiply_Float(
     A=&FloatFromCurve(FloatCurve=DataInterface() {propertyValues {Curve {keys [{interpMode=RCIM_Cubic, tangentMode=RCTM_User, time=0, value=0}, {interpMode=RCIM_Cubic, tangentMode=RCTM_User, time=1, value=1}]}}})
     B=&FloatFromCurve(
      FloatCurve=DataInterface() {propertyValues {Curve {keys [
          {arriveTangent=1.9704, interpMode=RCIM_Cubic, leaveTangent=1.9704, tangentMode=RCTM_User, time=0, value=0.2}
          {interpMode=RCIM_Cubic, tangentMode=RCTM_User, time=1, value=1.25}
         ]}}}
      CurveIndex=$Particles.RibbonLinkOrder
     )
    ))
   &Color(Color=&ColorFromCurve(
     ColorCurve=DataInterface() {propertyValues {
       AlphaCurve {keys [<at 0 2.5 RCIM_Cubic>,<at 0.8007 0 RCIM_Cubic>]}
       BlueCurve {keys [<at 0 2>,<at 0.5292 0.5>,<at 1 0.1 RCIM_Cubic>]}
       GreenCurve {keys [<at 0 1.45>,<at 0.5292 0.1587>,<at 1 0.0351 RCIM_Cubic>]}
       RedCurve {keys [<at 0 0.5>,<at 0.5292 0.025>,<at 1 0.005 RCIM_Cubic>]}
      }}
     CurveIndex=$Particles.RibbonLinkOrder
    ))
   /SetVariables(Particles.Offset X:NiagaraFloat=&Add_Float(A=$Particles.Offset X, B=&Lerp_Float(A=-0.02, B=-0.035, Alpha=$System.ContrailDiveAlpha))) {isSetParameters=true, module=SetVariables_#13}
   &DynamicMaterialParameters(
    Index 0 Param 1=&Add_Float(
     A=$Particles.Offset X
     B=&SelectFloatFromArray(
      Direct Array Index=$Particles.PartitionParticles.Partition
      Float Selection Array=DataInterface() {propertyValues {FloatData [0.23,0.46,0.56,0.96]}}
     )
    )
    Index 0 Param 2=$Particles.Offset X
    Index 0 Param 3=&SelectFloatFromArray(Direct Array Index=$Particles.PartitionParticles.Partition, Float Selection Array=DataInterface() {propertyValues {FloatData [1,1,1.5,1.5]}})
    Index 2 Param 1=1
    Module.Param0WriteEnabled=true
    Module.Param0WriteEnabled2=true
    Module.Param1WriteEnabled=false
    Module.Param1WriteEnabled2=false
    Module.Param2WriteEnabled=true
    Module.Param2WriteEnabled2=false
    Module.Param3WriteEnabled=false
    Module.Param3WriteEnabled2=false
    Write Parameter Index 2=true
   ) {module=DynamicMaterialParameters001}
   /SetVariables(Particles.Position=$Particles.InterpolatedPosition) {isSetParameters=true, module=SetVariables_#14}
  ]
  rendererClasses [&NiagaraRibbonRendererProperties]
  renderers [{rendererClass=&NiagaraRibbonRendererProperties}]
 }
 Decal_Glow {
  emitterUpdateScript [
   &EmitterState(
    Inactive Response=::NewEnumerator1
    Life Cycle Mode=::NewEnumerator1
    Loop Behavior=::NewEnumerator1
    Loop Duration Mode=::NewEnumerator1
    UseLoopDelay=false
   )
   &SpawnBurst_Instantaneous(Spawn Count=4, Spawn Probability=HlslExpression() {hlslExpression="Emitter.LoopCount == 0"}, Use Spawn Probability=false) {module=SpawnBurst_Instantaneous001}
  ]
  particleSpawnScript [
   &FN_SetSpawnGroup()
   &PartitionParticles(Number Of Partitions=4)
   &InitializeParticle(
    Color=<rgba 0 1.6667 2 1>
    Lifetime=3
    Sprite Size Mode=::NewEnumerator3
    Uniform Sprite Size=25
    UsePositionOffset=false
    Write Material Random=true
   ) {module=InitializeParticle001}
  ]
  particleUpdateScript [
   &ParticleState(Kill Particles When Lifetime Has Elapsed=false)
   &SocketLocation(
    Orientation Sampling=::NewEnumerator1
    SamplePosOffset=true
    Sampled Position Offset=&SelectVectorFromArray(
     Direct Array Index=$Particles.PartitionParticles.Partition
     Vector Selection Array=DataInterface() {propertyValues {FloatData [<0 0 30>, <0 0 30>, <0 0 0>, <0 0 0>]}}
    )
    Sampled Position Offset Coordinate Space:ENiagaraSkelMeshTransforms=::NewEnumerator2
    Socket Index=$Particles.PartitionParticles.Partition
    Socket Reader=$System.Socket Reader
    Socket Sampling Mode=::NewEnumerator3
    VelocitySpeedLimit=false
   )
   &ScaleColor(
    Scale Alpha=&FloatFromCurve(
     FloatCurve=DataInterface() {propertyValues {Curve {keys [{interpMode=RCIM_Cubic, tangentMode=RCTM_User, time=0, value=0}, {interpMode=RCIM_Cubic, tangentMode=RCTM_User, time=1, value=1}]}}}
     Scale Curve=&MaskFloatBySpawnGroup(
      Masked Float=0.8
      Passthrough Float=&MaskFloatBySpawnGroup(Masked Float=0.8, Passthrough Float=0.3, Spawn Group Mask=1)
     )
    )
    ScaleA=true
    ScaleRGB=true
   )
   /SetVariables(Particles.SpriteSize=&Multiply_Vector2D_ByFloat(Float=&Lerp_Float(A=0.8, B=1.6, Alpha=$System.Dive), Vector2D=$Particles.Initial.SpriteSize)) {isSetParameters=true}
   &FN_Light_Attributes(
    Color=$Particles.Initial.Color
    Create Color Attribute=true
    Create Exponent Attribute=true
    Create Radius Attribute=true
    Exponent=1
    Radius=&MaskFloatBySpawnGroup(
     Masked Float=35
     Passthrough Float=&MaskFloatBySpawnGroup(Masked Float=35, Passthrough Float=0, Spawn Group Mask=1)
    )
    Scale Color=5
   )
  ]
  rendererClasses [&NiagaraSpriteRendererProperties, &NiagaraLightRendererProperties]
  renderers [
   {rendererClass=&NiagaraSpriteRendererProperties}
   {rendererClass=&NiagaraLightRendererProperties, rendererIndex=1}
  ]
 }
 Paint_Ribbons {
  emitterSpawnScript [
   /SetVariables(Emitter.InterpolatedVector:Vector3f=<0 0 1>, Emitter.TargetVector:Vector3f=<0 0 1>) {isSetParameters=true, module=SetVariables_#5}
   /SetVariables(Emitter.RibbonLength=222) {isSetParameters=true, module=SetVariables_#6}
   /SetVariables(Emitter.ParticlesPerTrail=5) {isSetParameters=true, module=SetVariables_#7}
  ]
  emitterUpdateScript [
   &EmitterState(
    Inactive Response=::NewEnumerator1
    Life Cycle Mode=::NewEnumerator1
    Loop Behavior=::NewEnumerator1
    Loop Duration Mode=::NewEnumerator1
    UseLoopDelay=false
   )
   &SpawnBurst_Instantaneous(
    Spawn Count=&Multiply_Int(A=4, B=$Emitter.ParticlesPerTrail)
    Use Spawn Probability=false
    UseLoopCountLimit=false
   )
   &self:ScratchModule(Target Vector:Vector3f=$System.LookAtVector) {module=ScratchModule}
  ]
  particleSpawnScript [
   &InitializeParticle(
    Lifetime=2
    Position Mode=::NewEnumerator0
    Ribbon Width=15
    Ribbon Width Mode=::NewEnumerator1
    Sprite Size Mode=::NewEnumerator3
    Uniform Sprite Size=5
    Write Material Random=true
   )
   &PartitionParticles(
    Normalized Index In Partition Includes One=true
    Number Of Partitions=4
    Particles Per Partition:NiagaraInt32=$Emitter.ParticlesPerTrail
    Partition Mode:ENiagara_PartitionMode=::NewEnumerator1
    Write Partition Index to Ribbon Link Order:ENiagara_AttributeSamplingApplyOutput=::NewEnumerator0
    Write Partition to Ribbon ID=::NewEnumerator0
   )
   /SetVariables(Particles.RibbonWidth=&SelectFloatFromArray(Direct Array Index=$Particles.PartitionParticles.Partition, Float Selection Array=DataInterface() {propertyValues {FloatData [50,50,25,25]}})) {isSetParameters=true, module=SetVariables_#8}
   /SetVariables(Particles.RibbonLength:NiagaraFloat=&SelectFloatFromArray(
     Direct Array Index=$Particles.PartitionParticles.Partition
     Float Selection Array=DataInterface() {propertyValues {FloatData [220,220,100,100]}}
    )) {isSetParameters=true, module=SetVariables_#9}
  ]
  particleUpdateScript [
   &ParticleState(Kill Particles When Lifetime Has Elapsed=false)
   &SocketLocation(
    Orientation Sampling=::NewEnumerator1
    Position Sampling=::NewEnumerator1
    SamplePosOffset=true
    Sampled Position Offset=&SelectVectorFromArray(
     Direct Array Index=$Particles.PartitionParticles.Partition
     Vector Selection Array=DataInterface() {propertyValues {FloatData [<0 0 24>, <0 0 24>, <0 0 -5>, <0 0 -5>]}}
    )
    Sampled Position Offset Coordinate Space:ENiagaraSkelMeshTransforms=::NewEnumerator2
    Socket Index=$Particles.PartitionParticles.Partition
    Socket Reader=$System.Socket Reader
    Socket Sampling Mode=::NewEnumerator3
    VelocitySpeedLimit=false
   )
   /SetVariables(Particles.RibbonLength:NiagaraFloat=&Multiply_Float(
     A=$Particles.Initial.RibbonLength
     B=&FloatFromCurve(
      FloatCurve=DataInterface() {propertyValues {Curve {keys [{interpMode=RCIM_Cubic, tangentMode=RCTM_User, time=0, value=0}, {interpMode=RCIM_Cubic, tangentMode=RCTM_User, time=1, value=1}]}}}
      Scale Curve=&Lerp_Float(A=1, B=1.2, Alpha=$System.ContrailDiveAlpha)
     )
    )) {isSetParameters=true, module=SetVariables_#10}
   /SetVariables(Particles.InterpolatedTrailVector:Vector3f=$Emitter.InterpolatedVector) {isSetParameters=true, module=SetVariables_#11}
   /SetVariables(Particles.InterpolatedPosition:NiagaraPosition=&AddVectorToPosition(
     Position=$Output.SocketLocation.SampledPosition
     Vector=&Multiply_VectorByFloat(
      Vector=$Particles.InterpolatedTrailVector
      Float=&Multiply_Float(A=$Particles.RibbonLength, B=$Particles.PartitionParticles.Normalized Index In Partition)
     )
    )) {isSetParameters=true, module=SetVariables_#12}
   &ScaleRibbonWidth(Ribbon Width Scale=&Multiply_Float(
     A=&FloatFromCurve(
      FloatCurve=DataInterface() {propertyValues {Curve {keys [{interpMode=RCIM_Cubic, tangentMode=RCTM_User, time=0, value=0}, {interpMode=RCIM_Cubic, tangentMode=RCTM_User, time=1, value=1}]}}}
      Scale Curve=1.3
     )
     B=&MergeHelper_FloatSwitch(
      0=&FloatFromCurve(
       FloatCurve=DataInterface() {propertyValues {Curve {keys [
           {arriveTangent=2.2468, interpMode=RCIM_Cubic, leaveTangent=2.2468, tangentMode=RCTM_User, time=0, value=0.6}
           {arriveTangent=0.0009, interpMode=RCIM_Cubic, leaveTangent=0.0009, tangentMode=RCTM_User, time=0.3, value=1}
           {arriveTangent=-1.7785, interpMode=RCIM_Cubic, leaveTangent=-1.7785, tangentMode=RCTM_User, time=0.95, value=0}
          ]}}}
       CurveIndex=$Particles.RibbonLinkOrder
      )
      1=&FloatFromCurve(
       FloatCurve=DataInterface() {propertyValues {Curve {keys [
           {arriveTangent=2.2468, interpMode=RCIM_Cubic, leaveTangent=2.2468, tangentMode=RCTM_User, time=0, value=0.6}
           {arriveTangent=0.0009, interpMode=RCIM_Cubic, leaveTangent=0.0009, tangentMode=RCTM_User, time=0.3, value=1}
           {arriveTangent=-1.9602, interpMode=RCIM_Cubic, leaveTangent=-1.9602, tangentMode=RCTM_User, time=0.95, value=0}
          ]}}}
       CurveIndex=$Particles.RibbonLinkOrder
      )
      2=&FloatFromCurve(
       FloatCurve=DataInterface() {propertyValues {Curve {keys [
           {interpMode=RCIM_Cubic, tangentMode=RCTM_User, time=0.5, value=1}
           {arriveTangent=-2.9883, interpMode=RCIM_Cubic, leaveTangent=-2.9883, tangentMode=RCTM_User, time=1, value=0}
          ]}}}
       CurveIndex=$Particles.RibbonLinkOrder
      )
      3=&FloatFromCurve(
       FloatCurve=DataInterface() {propertyValues {Curve {keys [
           {interpMode=RCIM_Cubic, tangentMode=RCTM_User, time=0.5, value=1}
           {arriveTangent=-2.9883, interpMode=RCIM_Cubic, leaveTangent=-2.9883, tangentMode=RCTM_User, time=1, value=0}
          ]}}}
       CurveIndex=$Particles.RibbonLinkOrder
      )
      Index=$Particles.PartitionParticles.Partition
      Num=::NewEnumerator2
     )
    ))
   &Color(Color=&ColorFromCurve(
     ColorCurve=DataInterface() {propertyValues {
       AlphaCurve {keys [<at 0.4355 1.5 RCIM_Cubic>,<at 0.8993 0 RCIM_Cubic>]}
       BlueCurve {keys [<at 0.0111 1.5>,<at 0.5373 0.5>,<at 0.8276 0.1 RCIM_Cubic>]}
       GreenCurve {keys [<at 0.0111 1.3>,<at 0.5373 0.1587>,<at 0.8276 0.0351 RCIM_Cubic>]}
       RedCurve {keys [<at 0.0111 0>,<at 0.5373 0.025>,<at 0.8276 0.005 RCIM_Cubic>]}
      }}
     CurveIndex=$Particles.RibbonLinkOrder
    ))
   /SetVariables(Particles.Offset X:NiagaraFloat=&Add_Float(A=$Particles.Offset X, B=&Lerp_Float(A=0, B=-0.03, Alpha=$System.ContrailDiveAlpha))) {isSetParameters=true, module=SetVariables_#13}
   &DynamicMaterialParameters(
    Index 0 Param 1=&SelectFloatFromArray(
     Direct Array Index=$Particles.PartitionParticles.Partition
     Float Selection Array=DataInterface() {propertyValues {FloatData [0.23,0.49,0.67,0.88]}}
    )
    Index 0 Param 2=$Particles.Offset X
    Index 0 Param 3=&Multiply_Float(
     A=&FloatFromCurve(
      FloatCurve=DataInterface() {propertyValues {Curve {keys [
          {arriveTangent=0.0484, interpMode=RCIM_Cubic, leaveTangent=0.0484, tangentMode=RCTM_User, time=0.3, value=1}
          {arriveTangent=0.0924, interpMode=RCIM_Cubic, leaveTangent=0.0924, tangentMode=RCTM_User, time=0.5, value=1}
          <at 1 0 RCIM_Cubic>
         ]}}}
      CurveIndex=$Particles.PartitionParticles.Normalized Index In Partition
     )
     B=&SelectFloatFromArray(Direct Array Index=$Particles.PartitionParticles.Partition, Float Selection Array=DataInterface() {propertyValues {FloatData [2,1.3,0,0]}})
    )
    Module.Index0Param0Write=true
    Module.Index0Param1Write=true
    Module.Index0Param2Write=true
    Module.Index0Param3Write=true
   )
   /SetVariables(Particles.Position=$Particles.InterpolatedPosition) {isSetParameters=true, module=SetVariables_#14}
  ]
  rendererClasses [&NiagaraRibbonRendererProperties]
  renderers [{rendererClass=&NiagaraRibbonRendererProperties}]
 }
 Pen_Mesh {
  emitterSpawnScript [/SetVariables(Emitter.InterpolatedVector:Vector3f=<0 0 1>, Emitter.TargetVector:Vector3f=<0 0 1>) {isSetParameters=true}]
  emitterUpdateScript [
   &EmitterState(
    Inactive Response=::NewEnumerator1
    Life Cycle Mode=::NewEnumerator1
    Loop Behavior=::NewEnumerator1
    Loop Duration Mode=::NewEnumerator1
    UseLoopDelay=false
   )
   &SpawnBurst_Instantaneous(Spawn Count=2, Use Spawn Probability=false, UseLoopCountLimit=false)
   &self:ScratchModule_01(Rate=2, Target Vector:Vector3f=$System.LookAtVector) {module=ScratchModule_01}
  ]
  particleSpawnScript [
   &InitializeParticle(Lifetime=2, Mesh Scale Mode=::NewEnumerator3, Mesh Uniform Scale=0.3, UsePositionOffset=false, Write Material Random=true)
   &PartitionParticles(Particles Per Partition=1, Partition Mode:ENiagara_PartitionMode=::NewEnumerator1)
   &InitialMeshOrientation(
    Mesh Axis To Orient=<0 0 1>
    Mesh Orientation Mode=::NewEnumerator2
    Orientation Vector=&SelectVectorFromArray(
     Direct Array Index=$Particles.PartitionParticles.Partition
     Vector Selection Array=DataInterface() {propertyValues {FloatData [<0 1 1>, <0 -1 -1>]}}
    )
    Use Rotation Vector=false
   )
  ]
  particleUpdateScript [
   &ParticleState(Kill Particles When Lifetime Has Elapsed=false)
   &ScaleMeshSize(Scale Factor=&VectorFromFloat(Value=&FloatFromCurve(FloatCurve=DataInterface() {propertyValues {Curve {keys [
          {arriveTangent=1.4528, interpMode=RCIM_Cubic, leaveTangent=1.4528, tangentMode=RCTM_User, time=0, value=0}
          {interpMode=RCIM_Cubic, tangentMode=RCTM_User, time=1, value=1}
         ]}}})))
   /SetVariables(Particles.InitialRotation:Quat4f=&MultiplyQuaternion(
     Quaternion A=$Particles.Initial.MeshOrientation
     Quaternion B=&MakeQuatFromAxisAngle(
      Axis=<0 1 1>
      AngleInDegrees=&FloatFromCurve(
       FloatCurve=DataInterface() {propertyValues {Curve {keys [
           {arriveTangent=3.9037, interpMode=RCIM_Cubic, leaveTangent=3.9037, tangentMode=RCTM_User, time=0, value=0}
           {arriveTangent=-0.001, interpMode=RCIM_Cubic, leaveTangent=-0.001, tangentMode=RCTM_User, time=0.7, value=1.05}
           {arriveTangent=-0.0033, interpMode=RCIM_Cubic, leaveTangent=-0.0033, tangentMode=RCTM_User, time=1, value=1}
          ]}}}
       Scale Curve=&SelectFloatFromArray(Direct Array Index=$Particles.PartitionParticles.Partition, Float Selection Array=DataInterface() {propertyValues {FloatData [720,-720]}})
      )
     )
    )) {isSetParameters=true}
   &SocketLocation(
    SamplePosOffset=true
    Sampled Position Offset=&SelectVectorFromArray(
     Direct Array Index=$Particles.PartitionParticles.Partition
     Vector Selection Array=DataInterface() {propertyValues {FloatData [<0 0 30>, <0 0 30>]}}
    )
    Sampled Position Offset Coordinate Space:ENiagaraSkelMeshTransforms=::NewEnumerator2
    Socket Index=$Particles.PartitionParticles.Partition
    Socket Reader=$System.Socket Reader
    Socket Sampling Mode=::NewEnumerator3
    VelocitySpeedLimit=false
   )
  ]
  rendererClasses [&NiagaraMeshRendererProperties]
  renderers [{rendererClass=&NiagaraMeshRendererProperties, rendererIndex=0}]
 }
 Splash_Foam {
  emitterUpdateScript [
   &EmitterState(
    Inactive Response=::NewEnumerator1
    Life Cycle Mode=::NewEnumerator1
    Loop Behavior=::NewEnumerator1
    Loop Delay=1
    Loop Duration Mode=::NewEnumerator1
    UseLoopDelay=true
   )
   &SpawnRate(SpawnRate=&Lerp_Float(A=50, B=75, Alpha=$System.ContrailDiveAlpha), Use Spawn Probability=false)
  ]
  particleSpawnScript [
   &InitializeParticle(
    Color=<rgba 0.6 3 0.68 0.75>
    Lifetime=&Multiply_Float(
     A=&RandomRangeFloat(Minimum=0.7, Maximum=0.9)
     B=&Lerp_Float(A=1, B=0.6, Alpha=$System.ContrailDiveAlpha)
    )
    Mesh Scale Mode=::NewEnumerator3
    Sprite Rotation Mode=::NewEnumerator2
    Sprite Size Mode=::NewEnumerator1
    Uniform Sprite Size Max=30
    Uniform Sprite Size Min=20
    UsePositionOffset=false
    Write Material Random=true
   ) {module=InitializeParticle001}
   &PartitionParticles()
   /Niagara/Modules/Update/SubUV/SubUVAnimation(SubUV Animation Mode=::NewEnumerator1)
   &ShapeLocation(
    Cylinder Distribution=::NewEnumerator2
    Cylinder Height=1
    Cylinder Height Midpoint=1
    Cylinder Radius=&RandomRangeFloat(Minimum=10, Maximum=30)
    MaskSpawnGroup=false
    Rotation Axis=<0 0 1>
    Rotation Mode=::NewEnumerator1
    Shape Primitive=::NewEnumerator1
    U Position=&RandomRangeFloat(Minimum=0, Maximum=1)
    UseOwnerScale=true
    V Position=&RandomRangeFloat(Minimum=0, Maximum=1)
   )
   /SetVariables(Particles.VisibilityTag=&SelectIntFromArray(
     Array Sampling Mode=::NewEnumerator1
     Direct Array Index=?
     Int Selection Array=DataInterface() {propertyValues {IntData [0,1,0]}}
     Random Seed=0
    )) {isSetParameters=true}
   &SocketLocation(
    SamplePosOffset=true
    Sampled Position Offset=&SelectVectorFromArray(
     Direct Array Index=$Particles.PartitionParticles.Partition
     Vector Selection Array=DataInterface() {propertyValues {FloatData [<0 0 30>, <0 0 30>]}}
    )
    Sampled Position Offset Coordinate Space:ENiagaraSkelMeshTransforms=::NewEnumerator2
    Socket Index=$Particles.PartitionParticles.Partition
    Socket Reader=$System.Socket Reader
    Socket Sampling Mode=::NewEnumerator3
    VelocitySpeedLimit=false
   )
   &AddVelocity(
    Cone Angle=15
    Cone Axis=$System.LookAtVector
    Module.UseConeFalloff=false
    Module.UseVelDistribution=false
    Override Seed=false
    Velocity Mode=::NewEnumerator2
    Velocity Speed=&Multiply_Float(
     A=&RandomRangeFloat(Minimum=150, Maximum=200)
     B=&Lerp_Float(A=1, B=2, Alpha=$System.ContrailDiveAlpha)
    )
   ) {enabled=false}
  ]
  particleUpdateScript [
   &ParticleState()
   &Drag(Drag=2, Use Linear Drag=true, Use Rotational Drag=true) {enabled=false}
   &SolveForcesAndVelocity(Clamp Velocity=false, Limit Acceleration=false) {enabled=false}
   &self:FollowRibbon_01(
    Alpha:NiagaraFloat=&Add_Float(
     A=&RandomRangeFloat(Minimum=0, Maximum=0.15)
     B=&FloatFromCurve(
      FloatCurve=DataInterface() {propertyValues {Curve {keys [
          {arriveTangent=0.4215, interpMode=RCIM_Cubic, leaveTangent=0.4215, tangentMode=RCTM_User, time=0, value=0}
          {arriveTangent=1.5182, interpMode=RCIM_Cubic, leaveTangent=1.5182, tangentMode=RCTM_User, time=1, value=1}
         ]}}}
      Scale Curve=&RandomRangeFloat(Minimum=0.5, Maximum=0.6)
     )
    )
    ParticlePerRibbon=5
    RibbonNumber:NiagaraInt32=$Particles.PartitionParticles.Partition
    RibbonReader=DataInterface() {propertyValues {EmitterBinding {bindingMode=Other, emitterName=Paint_Ribbons}}}
   ) {module=FollowRibbon_01}
   /SetVariables(Particles.Position=$Output.FollowRibbon_01.RibbonPos) {isSetParameters=true}
   &ScaleSpriteSize(Scale Factor=&Vector2DFromFloat(Value=&FloatFromCurve(FloatCurve=DataInterface() {propertyValues {Curve {keys [{arriveTangent=1.3936, interpMode=RCIM_Cubic, leaveTangent=1.3936, tangentMode=RCTM_User, time=0, value=0.1}, <at 1 1 RCIM_Cubic>]}}})))
   &SpriteRotationRate(Rotation Rate=&FloatFromCurve(
     FloatCurve=DataInterface() {propertyValues {Curve {keys [
         {interpMode=RCIM_Cubic, tangentMode=RCTM_User, time=0, value=0}
         {arriveTangent=1.8494, interpMode=RCIM_Cubic, leaveTangent=1.8494, tangentMode=RCTM_User, time=1, value=1}
        ]}}}
     Scale Curve=&RandomRangeFloat(Minimum=-200, Maximum=200)
    ))
   &DynamicMaterialParameters(
    Index 0 Param 1=&RandomRangeFloat(Minimum=0, Maximum=1)
    Index 0 Param 3=&FloatFromCurve(FloatCurve=DataInterface() {propertyValues {Curve {keys [
         {arriveTangent=1.4815, interpMode=RCIM_Cubic, leaveTangent=1.4815, tangentMode=RCTM_User, time=0, value=0}
         {interpMode=RCIM_Cubic, tangentMode=RCTM_User, time=1, value=1}
        ]}}})
    Index 0 Param 4=&FloatFromCurve(FloatCurve=DataInterface() {propertyValues {Curve {keys [
         {arriveTangent=-0.0155, interpMode=RCIM_Cubic, leaveTangent=-0.0155, tangentMode=RCTM_User, time=0.3, value=0.05}
         {arriveTangent=-0.0133, interpMode=RCIM_Cubic, leaveTangent=-0.0133, tangentMode=RCTM_User, time=0.7, value=0.8}
        ]}}})
    Module.Param0WriteEnabled=true
    Module.Param1WriteEnabled=false
    Module.Param2WriteEnabled=true
    Module.Param3WriteEnabled=true
   )
   &ScaleColor(
    Scale Alpha=&FloatFromCurve(FloatCurve=DataInterface() {propertyValues {Curve {keys [
         {interpMode=RCIM_Cubic, tangentMode=RCTM_User, time=0.2, value=1}
         {arriveTangent=-1.2825, interpMode=RCIM_Cubic, leaveTangent=-1.2825, tangentMode=RCTM_User, time=1, value=0}
        ]}}})
    Scale RGB=&VectorFromFloat(Value=&FloatFromCurve(
      FloatCurve=DataInterface() {propertyValues {Curve {keys [
          {interpMode=RCIM_Cubic, tangentMode=RCTM_User, time=0, value=2}
          {arriveTangent=-0.9872, interpMode=RCIM_Cubic, leaveTangent=-0.9872, tangentMode=RCTM_User, time=1, value=0.25}
         ]}}}
      Scale Curve=&FN_FrontEndFloatModifier(Gameplay=1.2, Front End=2)
     ))
    ScaleA=true
    ScaleRGB=true
   )
   /SetVariables(Particles.SpriteSize02:Vector2f=&Multiply_Vector2D_ByFloat(Float=1.5, Vector2D=$Particles.SpriteSize)) {isSetParameters=true}
  ]
  rendererClasses [&NiagaraSpriteRendererProperties]
  renderers [
   {rendererClass=&NiagaraSpriteRendererProperties}
   {rendererClass=&NiagaraSpriteRendererProperties, rendererIndex=1}
  ]
 }
}
summary {
 systemName=NS_Contrail_IronLilac
 emitters [
  {emitterName=Paint_Ribbons, rendererClasses [&NiagaraRibbonRendererProperties]}
  {emitterName=Pen_Mesh, rendererClasses [&NiagaraMeshRendererProperties]}
  {emitterName=Decal_Glow, rendererClasses [&NiagaraSpriteRendererProperties, &NiagaraLightRendererProperties]}
  {emitterName=Splash_Foam, rendererClasses [&NiagaraSpriteRendererProperties]}
  {emitterName=Center_Ribbons, rendererClasses [&NiagaraRibbonRendererProperties]}
 ]
 userVariables [
  User.bIsFrontEnd {type=NiagaraBool}
  User.bIsFrontEndPreview {type=NiagaraBool}
  User.ParaGlideLeanAlpha {type=NiagaraFloat, defaultValue=0, description=""}
  User.RibbonIDOffset {type=NiagaraInt32}
 ]
}
